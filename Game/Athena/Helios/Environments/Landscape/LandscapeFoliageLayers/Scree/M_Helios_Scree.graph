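# Nanite-capable master material for instanced Helios landscape scree, with an IsGravel static switch choosing gravel or scree layers.
# Both layers combine biome tinting from terrain masks with world-aligned detail textures; gravel adds vertex-color tints and noise variation.
# An optional runtime virtual texture path blends into the landscape by height and per-instance fade, excluded on forward and mobile shading.
# Snow-tint and wetness branches are present, but their results feed nothing shown.
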Material /Game/Athena/Helios/Environments/Landscape/LandscapeFoliageLayers/Scree/M_Helios_Scree
graph {
 properties {
  bTangentSpaceNormal=false
  bUseMaterialAttributes=true
  bUsedWithInstancedStaticMeshes=true
  bUsedWithNanite=true
  pixelDepthOffsetMode=PDOM_Legacy
  usedShadingModels=Default Lit
 }
 nodes [
  n1 = Reroute()
  n2 = GetMaterialAttributes(None=@n1) {attributeGetTypes [#1, #2, #3]}
  n3 = ScalarParameter() {defaultValue=1, parameterName=Wet_AlbedoDarken}
  n4 = ScalarParameter() {defaultValue=0.5, parameterName=Wet_RoughnessMult}
  BlendMaterialAttributes(
   A=SetMaterialAttributes!(
    MaterialAttributes=@n2.MaterialAttributes
    Base Color=Multiply(A=@n2.Base Color, B=@n3)
    Specular=@n2.MaterialAttributes
    Roughness=Multiply(A=@n2.Roughness, B=@n4)
   ) {attributeSetTypes [#1, #2, #3]}
   B=@n1
   Alpha=NamedRerouteUsage()
  )
  Constant() {r=1}
  Constant() {r=5}
  n5 = MaterialFunctionCall!(
   Diffuse Texture Object=TextureObject() {samplerType=SAMPLERTYPE_Masks, texture=/Game/Athena/WinterArt/Textures/T_DirtLeaks_M}
   Spec Texture Object
   TextureSize=ScalarParameter() {defaultValue=64, group=Scree, parameterName=Scree_DetailNoiseScale}
   --------------------------
   Use seperate settings for Z projection
   Use seperate texture for Z projection
   Texture Size Z projection
   Z Plane Diffuse Texture Object
   Z Plane Spec Texture Object
   "--------------------------      "
   ProjectionTransitionContrast
   Use Texture Alpha Heightmap to Refine the Transistion
   HeightLerpTransitionContrast
   PickTheDiffuseHeightChannel
   TheDiffuseHeightChannel
   "--------------------------       "
   NormalTextureObject=TextureObject() {samplerType=SAMPLERTYPE_Normal, texture=/Game/Environments/Apollo/Rocks/Natural_Rocks/Textures/T_RockDetails_02_N}
   World Space Normals=StaticBool() {value=true}
   "--------------------------  "
   Export Float 4 Diffuse
   Export Float 4 Spec
   "--------------------------     "
   WorldPosition
   World Space Normal
  ) {materialFunction=/Engine/Functions/Engine_MaterialFunctions02/Texturing/WorldAlignedTextures_Complex}
  Multiply(A=@n5.Normal Z Texture, B=VectorParameter() {group=Scree, parameterName=DetailNormalStrength, defaultValue=<rgba 2 2 0 1>}.RGB)
  n6 = MaterialFunctionCall(
   Diffuse Texture Object=TextureObject() {samplerType=SAMPLERTYPE_Masks, texture=&T_DirtLeaks_M}
   TextureSize=ScalarParameter() {defaultValue=64, group=Gravel, parameterName=Gravel_DetailNoiseScale}
   NormalTextureObject=TextureObject() {samplerType=SAMPLERTYPE_Normal, texture=&T_RockDetails_02_N}
   World Space Normals=StaticBool() {value=true}
  ) {materialFunction=&WorldAlignedTextures_Complex}
  NamedRerouteDeclaration(None=@n6.Normal Z Texture) {name=Details N_1, nodeColor=<rgba 0 1 0.5078 1>}
  n7 = MaterialFunctionCall(Float3=@n5.Z Diffuse Texture) {materialFunction=&BreakOutFloat3Components}
  NamedRerouteDeclaration(None=@n7.B) {name=Details-B, nodeColor=<rgba 0.4062 1 0 1>}
  n8 = TextureSampleParameter2D() {
   group=Textures
   parameterName=Texture_Specular
   samplerType=SAMPLERTYPE_Masks
   texture=/Game/Global/Textures/Default/Blanks/Temp_specular_Mask
  }
  NamedRerouteDeclaration(None=@n8.R) {name=Texture_Specular, nodeColor=<rgba 0 0.8359 1 1>}
  NamedRerouteDeclaration(None=@n8.B) {name=Texture_Roughness, nodeColor=<rgba 1 0.375 0 1>}
  n9 = TextureSampleParameter2D() {group=Textures, parameterName=Texture_Normal, samplerType=SAMPLERTYPE_Normal, texture=&FlatNormalBlue}
  NamedRerouteDeclaration(None=@n9.RGB) {name=Texture_Normal, nodeColor=<rgba 1 0.7266 0 1>}
  n10 = MaterialFunctionCall(Float3=@n6.Z Diffuse Texture) {materialFunction=&BreakOutFloat3Components}
  NamedRerouteDeclaration(None=@n10.G) {name=Details-G_1, nodeColor=<rgba 0.6172 0 1 1>}
  n11 = DotProduct(A=Constant3Vector() {constant=<rgba 0 0 1 0>}, B=VertexNormalWS())
  NamedRerouteDeclaration(None=SphereMask(
    A=ConstantBiasScale(None=@n11)
    B=ScalarParameter() {defaultValue=1, group=Snow, parameterName=G_SnowSample}
    Radius=ScalarParameter() {defaultValue=0.3, group=Snow, parameterName=G_SnowSampleRadius}
    Hardness=ScalarParameter() {defaultValue=0.5, group=Snow, parameterName=G_SnowSampleHardness}
   )) {name=G_SnowTop_Mask, nodeColor=<rgba 0.1953 0 1 1>}
  NamedRerouteDeclaration(None=@n10.B) {name=Details-B_1, nodeColor=<rgba 0.4062 1 0 1>}
  n12 = WorldPosition() {worldPositionShaderOffset=WPT_ExcludeAllShaderOffsets}
  n13 = ScalarParameter() {defaultValue=64, parameterName=WaterLevelOffset}
  n14 = Add(
   A=Add(A=ComponentMask(None=@n12.XYZ) {b=true}, B=NamedRerouteDeclaration(None=@n13) {name=WaterLevelOffset, nodeColor=<rgba 1 0 1 1>})
   B=Constant() {desc=slightly wet sand height offset, r=10}
  )
  n15 = ScalarParameter() {defaultValue=90, parameterName=WaterLevelTransition}
  NamedRerouteDeclaration(None=Saturate(None=Divide(A=@n14, B=@n15))) {name=OceanMask, nodeColor=<rgba 0.5 0 1 1>}
  NamedRerouteDeclaration(None=@n5.Normal Z Texture) {name=Details N, nodeColor=<rgba 0 1 0.5078 1>}
  n16 = NamedRerouteUsage()
  n17 = RuntimeVirtualTextureSampleParameter() {materialType=BaseColor_Normal_Specular_YCoCg, parameterName=RVT, virtualTexture=/Game/Athena/Apollo/Maps/Landscape/VT/RVT_Apollo_Landscape}
  n18 = SetMaterialAttributes!(
   MaterialAttributes
   Base Color=@n17.BaseColor
   Specular=@n17.BaseColor
   Roughness=@n17.BaseColor
   Emissive Color=Constant()
   Normal=@n17.BaseColor
  ) {attributeSetTypes [
    #1
    #2
    #3
    #4
    #5
   ]}
  n19 = NamedRerouteUsage()
  n20 = GetMaterialAttributes(None=@n19) {attributeGetTypes [#5]}
  n21 = WorldPosition()
  n22 = RuntimeVirtualTextureSampleParameter() {materialType=WorldHeight, parameterName=RVTHeight, virtualTexture=/Game/Athena/Apollo/Maps/Landscape/VT/RVT_Apollo_Landscape_Height}
  n23 = Subtract(A=Add(A=ScalarParameter() {group=RVT, parameterName=Offset RVT WP}, B=ComponentMask(None=@n21.XYZ) {b=true}), B=@n22.WorldHeight)
  n24 = ScalarParameter() {defaultValue=1, group=RVT, parameterName=Blend RVT}
  n25 = LinearInterpolate(
   A=Multiply(A=ScalarParameter() {defaultValue=1, group=RVT, parameterName=RVT Scale Distance Blend}, B=@n24) {constA=3}
   B=@n24
   Alpha=PerInstanceFadeAmount()
  )
  n26 = BlendMaterialAttributes(
   A=@n18
   B=StaticSwitchParameter(
    True=SetMaterialAttributes!(MaterialAttributes=@n18, Normal=@n20.Normal) {attributeSetTypes [#5]}
    False=@n19
   ) {parameterName=RVT Colors}
   Alpha=Saturate(None=Divide(A=@n23, B=@n25))
  ) {vertexAttributeBlendType=UseA}
  NamedRerouteDeclaration(None=MaterialFunctionCall(NoRVT=@n16, RVT=MaterialProxyReplace(Realtime=@n26, MaterialProxy=@n16)) {materialFunction=&FN_RVT_Switch_Material}.Result) {name=WithRVT, nodeColor=<rgba 1 0 0.2109 1>}
  n27 = TextureObject() {samplerType=SAMPLERTYPE_Grayscale, texture=/Game/Weapons/FORT_Melee/Pickaxe_Cube_Ninja_Male/FX/T_CellNoise}
  n28 = ScalarParameter() {defaultValue=300, group=Gravel, parameterName=Gravel_VariationTextureScale}
  n29 = Multiply(
   A=ComponentMask(None=MaterialFunctionCall(TextureObject=@n27, TextureSize=@n28) {materialFunction=/Game/Weapons/WeaponSkins/Wraps/MaterialFunctions/WorldAlignedTexture}.Z Texture) {r=true}
   B=ScalarParameter() {defaultValue=5, group=Gravel, parameterName=Gravel_VariationTextureMod}
  )
  n30 = ScalarParameter() {defaultValue=1, group=Gravel, parameterName=Gravel_VariationTexturePow}
  NamedRerouteDeclaration(None=Saturate(None=Power(Base=@n29, Exp=@n30))) {name=Gravel_Noise_Mask, nodeColor=<rgba 1 0 0.6562 1>}
  n31 = VectorParameter() {group=Gravel, parameterName=Gravel_Base-MacroHueShift, sortPriority=-150, defaultValue=<rgba 0.0704 0.0592 0.0513 1>}
  n32 = VectorParameter() {group=Gravel, parameterName=Gravel_BaseColor, sortPriority=-500, defaultValue=<rgba 0.0434 0.04 0.0303 0>}
  n33 = VectorParameter() {group=Gravel, parameterName=Gravel_BaseColor_VertexRed, sortPriority=-400, defaultValue=<rgba 0.0605 0.0747 0.0691 0>}
  n34 = VertexColor()
  n35 = PerInstanceRandom()
  n36 = MaterialFunctionCall(Float3=LinearInterpolate(
    A=@n34
    B=MaterialFunctionCall(Hue Shift Percentage=@n35, Texture=@n34) {materialFunction=&HueShift}.Result
    Alpha=ScalarParameter() {group=Gravel, parameterName=Gravel_UseVertexColorMaskHueShift, sortPriority=-100}
   )) {materialFunction=&BreakOutFloat3Components}
  n37 = LinearInterpolate(A=@n32.RGB, B=@n33.RGB, Alpha=@n36.R)
  n38 = VectorParameter() {group=Gravel, parameterName=Gravel_BaseColor_VertexGreen, sortPriority=-300, defaultValue=<rgba 0.1372 0.0931 0.0593 0>}
  n39 = VectorParameter() {group=Gravel, parameterName=Gravel_BaseColor_VertexBlue, sortPriority=-200, defaultValue=<rgba 0.0677 0.0557 0.0469 0>}
  n40 = LinearInterpolate(
   A=@n32.RGB
   B=LinearInterpolate(A=LinearInterpolate(A=@n37, B=@n38.RGB, Alpha=@n36.G), B=@n39.RGB, Alpha=@n36.B)
   Alpha=ScalarParameter() {defaultValue=1, group=Gravel, parameterName=Gravel_UseVertexColorMask, sortPriority=-100}
  )
  n41 = Saturate(None=Multiply(A=ScalarParameter() {defaultValue=2, group=Gravel, parameterName=Gravel_BaseMacroHueMod, sortPriority=-125}, B=Multiply(A=@n31.RGB, B=@n40)))
  n42 = NamedRerouteUsage()
  n43 = MaterialFunctionCall!(UseLandscapeCoords, "UseProjection(translucent)") {materialFunction=/Game/Athena/Helios/Environments/Landscape/Materials/MaterialFunctions_Utilities/MF_TerrainMasks}
  n44 = MaterialFunctionCall!(
   Material=SetMaterialAttributes!(MaterialAttributes, Base Color=LinearInterpolate(A=@n41, B=@n40, Alpha=@n42)) {attributeSetTypes [#1]}
   TerrainMasks_QuadA=@n43.QuadA_RGBA
   TerrainMasks_QuadB=@n43.QuadA_RGBA
   TerrainMasks_QuadC=@n43.QuadA_RGBA
   TerrainMasks_QuadD=@n43.QuadA_RGBA
  ) {materialFunction=/Game/Athena/Helios/Environments/Landscape/Materials/MaterialFunctions_Utilities/MF_Gravel_BiomeColorTint}
  n45 = ScalarParameter() {defaultValue=0.25, group=Gravel, parameterName=Gravel_SpecMin}
  n46 = ScalarParameter() {defaultValue=1, group=Gravel, parameterName=Gravel_SpecMax}
  n47 = NamedRerouteUsage()
  n48 = ScalarParameter() {defaultValue=0.1, group=Gravel, parameterName=Gravel_RoughnessMin}
  n49 = ScalarParameter() {defaultValue=1, group=Gravel, parameterName=Gravel_RoughnessMax}
  n50 = NamedRerouteUsage()
  n51 = SetMaterialAttributes!(
   MaterialAttributes
   Base Color=MaterialFunctionCall(
    Base=GetMaterialAttributes(None=@n44.Result) {attributeGetTypes [#1]}.Base Color
    Blend=NamedRerouteUsage()
   ) {materialFunction=&Blend_Overlay}.Result
   Specular=Saturate(None=LinearInterpolate(A=@n45, B=@n46, Alpha=@n47))
   Roughness=Saturate(None=LinearInterpolate(A=@n48, B=@n49, Alpha=@n50))
   Normal=NamedRerouteUsage()
  ) {attributeSetTypes [
    #1
    #2
    #3
    #5
   ]}
  NamedRerouteDeclaration(None=@n51) {name=Gravel_Material, nodeColor=<rgba 1 0 0.2109 1>}
  NamedRerouteDeclaration(None=@n10.R) {name=Details-R_1, nodeColor=<rgba 0 1 0.5078 1>}
  NamedRerouteDeclaration(None=@n7.R) {name=Details-R, nodeColor=<rgba 0 1 0.5078 1>}
  n52 = NamedRerouteUsage()
  n53 = ComponentMask(None=VectorParameter() {group=Scree, parameterName=Scree_Basecolor, defaultValue=<rgba 0.125 0.125 0.125 1>}.RGB) {b=true, g=true, r=true}
  n54 = MaterialFunctionCall() {materialFunction=&MF_TerrainMasks}
  n55 = MaterialFunctionCall!(
   Material=SetMaterialAttributes!(MaterialAttributes, Base Color=StaticSwitchParameter(True=@n52, False=@n53) {parameterName=UseDiffuseTexture}) {attributeSetTypes [#1]}
   TerrainMasks_QuadA=@n54.QuadA_RGBA
   TerrainMasks_QuadB=@n54.QuadA_RGBA
   TerrainMasks_QuadC=@n54.QuadA_RGBA
   TerrainMasks_QuadD=@n54.QuadA_RGBA
  ) {materialFunction=/Game/Athena/Helios/Environments/Landscape/Materials/MaterialFunctions_Utilities/MF_Scree_BiomeColorTint}
  n56 = SetMaterialAttributes!(
   MaterialAttributes
   Base Color=MaterialFunctionCall(
    Base=GetMaterialAttributes(None=@n55.Result) {attributeGetTypes [#1]}.Base Color
    Blend=NamedRerouteUsage()
   ) {materialFunction=&Blend_Overlay}.Result
   Specular=LinearInterpolate(
    A=ScalarParameter() {defaultValue=0.25, group=Scree, parameterName=Scree_Spec_Min}
    B=ScalarParameter() {defaultValue=1, group=Scree, parameterName=Scree_Spec_Max}
    Alpha=NamedRerouteUsage()
   )
   Roughness=LinearInterpolate(
    A=ScalarParameter() {defaultValue=0.1, group=Scree, parameterName=Scree_Rough_Min}
    B=ScalarParameter() {defaultValue=1, group=Scree, parameterName=Scree_Rough_Max}
    Alpha=NamedRerouteUsage()
   )
   Normal=VertexNormalWS()
  ) {attributeSetTypes [
    #1
    #2
    #3
    #5
   ]}
  NamedRerouteDeclaration(None=@n56) {name=Scree_Material, nodeColor=<rgba 1 0 0.2109 1>}
  NamedRerouteDeclaration(None=TextureSampleParameter2D() {group=Textures, parameterName=Texture_BaseColor, texture=/Game/Global/Textures/Default/Blanks/T_MidgreysRGB}.RGB) {name=Texture_BaseColor, nodeColor=<rgba 0 1 0.1562 1>}
  NamedRerouteDeclaration(None=@n7.G) {name=Details-G, nodeColor=<rgba 0.6172 0 1 1>}
  n57 = StaticSwitchParameter(True=NamedRerouteUsage(), False=NamedRerouteUsage()) {parameterName=IsGravel}
  Reroute(None=@n57)
  n58 = GetMaterialAttributes(None=@n57) {attributeGetTypes [#1]}
  n59 = MaterialFunctionCall!(
   ScalarWeightMask
   UseSnowOcclusionTextureMask
   ------
   UseMaterialInput
   AdditionalMaskToRemoveSnow
   OutputTangentNormals
   UseAdditionalMaskToRemoveSnow
   UseSnowNormals
   OuputVertexNormalsAsTheSnowNormals
   UseSparkle
   LQSpark
   AllowSnowMicroNormals
   CliffMaterialOnly
   Material
   CliffLandscapeLayerMask
   IsLandscape
   AreInputNormalsTangentSpace
   UseInputNormalsForFakeSnowShadow
  ) {materialFunction=/Game/Athena/Helios/Environments/Landscape/Materials/QualityLevelLayers/MF_Helios_Snow_Default}
  n60 = NamedRerouteUsage()
  SetMaterialAttributes!(
   MaterialAttributes=@n58.MaterialAttributes
   Base Color=LinearInterpolate(
    A=@n58.Base Color
    B=VectorParameter() {group=Snow, parameterName=G_SnowColor, defaultValue=<rgba 0.4292 0.4575 0.5 1>}.RGB
    Alpha=Multiply(A=@n59.DetailMask, B=@n60)
   )
  ) {attributeSetTypes [#1]}
  n61 = NamedRerouteUsage()
  n62 = StaticSwitchParameter(True=NamedRerouteUsage(), False=@n61) {parameterName=UseRVT}
  ShadingPathSwitch(Default=@n62, Deferred=@n62, Forward=@n61, Mobile=@n61)
 ]
}
